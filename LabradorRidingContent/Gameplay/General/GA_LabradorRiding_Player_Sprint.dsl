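; Gameplay ability that lets a player riding a creature sprint the mount.
; Activation holsters the weapon, applies the creature's sprint effect, flags it sprinting and plays audio; ending undoes these.
; A hotfix toggle enables the mount's collision reaction, which damages non-teammate pawns hit and fires an impact GameplayCue.
; Tag-wait tasks remove the rider's sprint effect when any cancel tag is added, rather than ending the ability directly.
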(event Ability|EventActivateAbility
  (bind _returnvalue (Ability|GetAvatarActorfromActorInfo))
  (bind _returnvalue_2 (Ability|GetAvatarActorfromActorInfo))
  (bind _returnvalue_3 (Actor|GetComponentbyClass _returnvalue_2 "/Script/RidingCodeRuntime.ControllingRiderComponent"))
  (bind _returnvalue_4 (Actor|GetComponentbyClass _returnvalue "/Script/RidingCodeRuntime.ControllingRiderComponent"))
  (bind _donotuse_activeridable (Variables|/Script/RidingCodeRuntime.RiderComponent|GetDoNotUse_ActiveRidable _returnvalue_4))
  (bind _returnvalue_5 (Components|GetOwner _donotuse_activeridable))
  (bind _returnvalue_6 (Actor|GetComponentbyClass _returnvalue_5 "/Script/FortniteGame.FortAthenaCollisionReactionComponent"))
  (bind _donotuse_activeridable_2 (Variables|/Script/RidingCodeRuntime.RiderComponent|GetDoNotUse_ActiveRidable _returnvalue_3))
  (bind _returnvalue_7 (Components|GetOwner _donotuse_activeridable_2))
  (bind _returnvalue_8 (Ability|GetAbilitySystemComponent _returnvalue_2))
  (Utilities|FlowControl|Sequence
    (:then_0
      (if (ScalableFloat|GetValueasBool (Variables|Default|GetHF_PawnKnockbackEnabled))
        (Utilities|IsValid _returnvalue_6
          (:Is%20Valid
            (Interaction|BindEventtoOnReaction _returnvalue_6 (EventDispatchers|CreateEvent :!event CollisionReaction))))))
    (:then_1
      (Ability|CommitAbility)
      (Utilities|IsValid _returnvalue_2
        (:Is%20Valid
          (Utilities|IsValid _returnvalue_3
            (:Is%20Valid
              (Utilities|IsValid _donotuse_activeridable_2
                (:Is%20Valid
                  (Utilities|IsValid _returnvalue_7
                    (:Is%20Valid
                      (bind _asfort_pawn (Utilities|Casting|CastToFortPawn _returnvalue_2))
                      (Utilities|IsValid _asfort_pawn
                        (:Is%20Valid
                          (Weapon|SetHolsterWeapon _asfort_pawn 0 true false "Riding")
                          (bind _aslabrador_riding_creature_base_ridable_component (Utilities|Casting|CastToLabradorRiding_CreatureBaseRidableComponent _donotuse_activeridable_2))
                          (Class|ControllableRidableComponent|CancelBuildBar _aslabrador_riding_creature_base_ridable_component)
                          (Creative|GameplayEffects|SimpleApplyGameplayEffectToActor (Variables|/LabradorRidingContent/Components/LabradorRiding_CreatureBaseRidableComponent.LabradorRiding_CreatureBaseRidableComponent_C|GetGE_Creature_Sprint _aslabrador_riding_creature_base_ridable_component) _returnvalue_7)
                          (Networking|ForceNetUpdate _returnvalue_7)
                          (Variables|/Script/RidingCodeRuntime.RidableComponent|SetbIsSprinting true _donotuse_activeridable_2)
                          (Utilities|FlowControl|Sequence
                            (:then_0
                              (Default|StartCollisionReaction))
                            (:then_1
                              (for _array_element (Variables|Default|GetTags%20to%20cancel%20sprinting)
                                (Ability|Tasks|WaitGameplayTagAdd _array_element
                                  (:Added
                                    (Utilities|IsValid _returnvalue_8
                                      (:Is%20Valid
                                        (GameplayEffects|RemoveActiveGameplayEffectbySourceEffect _returnvalue_8 "/LabradorRidingContent/Gameplay/General/GE_LabradorRiding_Player_Sprint.GE_LabradorRiding_Player_Sprint_C" 0 -1)))))))
                            (:then_2
                              (Default|PlaySoundLib (Utilities|Casting|CastToNPC_Pawn_Wildlife_Parent (Components|GetOwner (Variables|/Script/RidingCodeRuntime.RiderComponent|GetDoNotUse_ActiveRidable (Actor|GetComponentbyClass (Ability|GetAvatarActorfromActorInfo) "/Script/RidingCodeRuntime.ControllingRiderComponent")))) (Variables|Default|GetSprintStartSoundLibTag)))))))))))))))
    (:then_2
      (Audio|PlaySound2D "/Labrador/Sound/Foley/SC_Labrador_SpeedBoost_Whip.SC_Labrador_SpeedBoost_Whip" 1.0 1.0 0.0 0 (Ability|GetAvatarActorfromActorInfo))
      (Variables|Default|GetSprintSound
        (:then
          (Audio|Components|Audio|Play (Variables|Default|GetSprintSound) 0.0)
          (Audio|Components|Audio|FadeIn (Variables|Default|GetSprintSound) 0.1 1.0 0.0 "Linear" :!class /Script/Engine.AudioComponent))
        (:else
          (bind _returnvalue_9 (Audio|SpawnSound2D "/Labrador/Sound/Foley/SC_Labrador_SpeedBoost_LP.SC_Labrador_SpeedBoost_LP"))
          (Variables|Default|SetSprintSound _returnvalue_9)
          (Audio|Components|Audio|Play (Variables|Default|GetSprintSound) 0.0)
          (Audio|Components|Audio|FadeIn (Variables|Default|GetSprintSound) 0.1 1.0 0.0 "Linear" :!class /Script/Engine.AudioComponent))))))

(event Ability|EventFailedToActivatePassiveAbility
  (params (ActorInfo :type Gameplay%20Ability%20Actor%20Info%20Structure)))

(event Ability|EventOnEndAbility
  (params (bWasCancelled :type Boolean))
  (bind _returnvalue (Ability|GetAvatarActorfromActorInfo))
  (bind _returnvalue_2 (Actor|GetComponentbyClass _returnvalue "/Script/RidingCodeRuntime.ControllingRiderComponent"))
  (bind _donotuse_activeridable (Variables|/Script/RidingCodeRuntime.RiderComponent|GetDoNotUse_ActiveRidable _returnvalue_2))
  (bind _returnvalue_3 (Components|GetOwner _donotuse_activeridable))
  (bind _returnvalue_4 (Ability|GetAbilitySystemComponent _returnvalue_3))
  (Default|EndCollisionReaction)
  (Audio|Components|Audio|FadeOut (Variables|Default|GetSprintSound) 1.0 0.0 "Linear" :!class /Script/Engine.AudioComponent)
  (Utilities|IsValid _returnvalue
    (:Is%20Valid
      (bind _asfort_pawn (Utilities|Casting|CastToFortPawn _returnvalue))
      (Utilities|IsValid _asfort_pawn
        (:Is%20Valid
          (Utilities|IsValid _returnvalue_2
            (:Is%20Valid
              (Utilities|IsValid _donotuse_activeridable
                (:Is%20Valid
                  (Utilities|IsValid _returnvalue_3
                    (:Is%20Valid
                      (Weapon|SetHolsterWeapon _asfort_pawn 0 false true "Riding")
                      (Utilities|IsValid _returnvalue_4
                        (:Is%20Valid
                          (bind _aslabrador_riding_creature_base_ridable_component (Utilities|Casting|CastToLabradorRiding_CreatureBaseRidableComponent _donotuse_activeridable))
                          (GameplayEffects|RemoveActiveGameplayEffectbySourceEffect _returnvalue_4 (Variables|/LabradorRidingContent/Components/LabradorRiding_CreatureBaseRidableComponent.LabradorRiding_CreatureBaseRidableComponent_C|GetGE_Creature_Sprint _aslabrador_riding_creature_base_ridable_component) 0 -1)
                          (Networking|ForceNetUpdate _returnvalue_3)
                          (Variables|/Script/RidingCodeRuntime.RidableComponent|SetbIsSprinting false _donotuse_activeridable)
                          (Utilities|Casting|CastToLabradorRiding_CreatureBaseRidableComponent _donotuse_activeridable))))))))))))))

(event Custom|ClientShowDamageNumber
  (params (Target :type Fort%20Pawn%20Object%20Reference) (Location :type Vector))
  (FortPawn|GameplayCueDamage Target "Executed" (Ability|GameplayCue|MakeGameplayCueParameters :NormalizedMagnitude 0.0 :RawMagnitude (ScalableFloat|GetValueatLevel (Variables|Default|GetHF_PawnKnockbackDamage)) :EffectContext (Class|AthenaFunctionLibrary|MakeSimpleFortEffectContext (CallFunction|GetAvatarasFortPlayerPawn) (Components|GetOwner (Variables|/Script/RidingCodeRuntime.RiderComponent|GetDoNotUse_ActiveRidable (Actor|GetComponentbyClass (Ability|GetAvatarActorfromActorInfo) "/Script/RidingCodeRuntime.ControllingRiderComponent")))) :Location Location :Normal "0,0,0" :SourceObject self :GameplayEffectLevel 1 :AbilityLevel 1 :bReplicateLocationWhenUsingMinimalRepProxy false)))

(fn CollisionReaction
  (params (HitActor :type Actor%20Object%20Reference) (bFromAsyncSweepBox :type Boolean) (CollisionReactionType :type FCollisionReactionType%20Enum :default "Nothing") (bIsFirstContinuousReactionOnDelayableActor :type Boolean) (HitResult :type Hit%20Result%20Structure%20%28by%20ref%29))
  (bind _returnvalue (CallFunction|GetAvatarasFortPlayerPawn))
  (bind (_bblockinghit _binitialoverlap _time _distance _location _impactpoint _normal _impactnormal _physmat) (Collision|BreakHitResult HitResult))
  (bind _asfort_pawn (Utilities|Casting|CastToFortPawn HitActor))
  (if (not (Team|OnSameTeam _asfort_pawn _returnvalue))
    (GameplayEffects|FortApplyGameplayEffectToSelf :self (Ability|GetFortAbilitySystemComponent _asfort_pawn) :GameplayEffectClass "/Labrador/Abilities/GE_Labrador_GenericDamage.GE_Labrador_GenericDamage_C" :OptionalSourceObject self :OptionalEffectCauser (CallFunction|GetAvatarasFortPlayerPawn) :OptionalDamageSource (Components|GetOwner (Variables|/Script/RidingCodeRuntime.RiderComponent|GetDoNotUse_ActiveRidable (Actor|GetComponentbyClass (Ability|GetAvatarActorfromActorInfo) "/Script/RidingCodeRuntime.ControllingRiderComponent"))) :Level 1.0))
  (GameplayCue|ExecuteGameplayCueOnActor(Burst) (CallFunction|GetAvatarasFortPlayerPawn) (Variables|Default|GetSprintImpactGameplayCueTag) (Ability|GameplayCue|MakeGameplayCueParameters :NormalizedMagnitude (Math|Conversions|ToFloat(Boolean) (Team|OnSameTeam _asfort_pawn _returnvalue)) :RawMagnitude (ScalableFloat|GetValueatLevel (Variables|Default|GetHF_PawnKnockbackDamage)) :EffectContext (Class|AthenaFunctionLibrary|MakeSimpleFortEffectContext (CallFunction|GetAvatarasFortPlayerPawn) (Components|GetOwner (Variables|/Script/RidingCodeRuntime.RiderComponent|GetDoNotUse_ActiveRidable (Actor|GetComponentbyClass (Ability|GetAvatarActorfromActorInfo) "/Script/RidingCodeRuntime.ControllingRiderComponent")))) :Location _location :Normal _location :Instigator _asfort_pawn :SourceObject self :PhysicalMaterial _physmat :GameplayEffectLevel 1 :AbilityLevel 1 :bReplicateLocationWhenUsingMinimalRepProxy false)))

(fn EndCollisionReaction
  (bind _returnvalue (Ability|GetAvatarActorfromActorInfo))
  (bind _returnvalue_2 (Actor|GetComponentbyClass _returnvalue "/Script/RidingCodeRuntime.ControllingRiderComponent"))
  (bind _donotuse_activeridable (Variables|/Script/RidingCodeRuntime.RiderComponent|GetDoNotUse_ActiveRidable _returnvalue_2))
  (bind _returnvalue_3 (Components|GetOwner _donotuse_activeridable))
  (bind _returnvalue_4 (Actor|GetComponentbyClass _returnvalue_3 "/Script/FortniteGame.FortAthenaCollisionReactionComponent"))
  (if (ScalableFloat|GetValueasBool (Variables|Default|GetHF_PawnKnockbackEnabled))
    (Utilities|IsValid _donotuse_activeridable
      (:Is%20Valid
        (Utilities|IsValid _returnvalue_4
          (:Is%20Valid
            (CollisionReactions|Stop _returnvalue_4)))))))

(fn OnRep_CancelSprintDispatcher)

(fn StartCollisionReaction
  (bind _returnvalue (Ability|GetAvatarActorfromActorInfo))
  (bind _returnvalue_2 (Actor|GetComponentbyClass _returnvalue "/Script/RidingCodeRuntime.ControllingRiderComponent"))
  (bind _donotuse_activeridable (Variables|/Script/RidingCodeRuntime.RiderComponent|GetDoNotUse_ActiveRidable _returnvalue_2))
  (bind _returnvalue_3 (Components|GetOwner _donotuse_activeridable))
  (bind _returnvalue_4 (Actor|GetComponentbyClass _returnvalue_3 "/Script/FortniteGame.FortAthenaCollisionReactionComponent"))
  (if (ScalableFloat|GetValueasBool (Variables|Default|GetHF_PawnKnockbackEnabled))
    (Utilities|IsValid _returnvalue_4
      (:Is%20Valid
        (CollisionReactions|Run _returnvalue_4)))))
